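# Example translucent, unlit particle sprite material that demonstrates the 3dParticleOpacity material function for giving sprites apparent depth.
# Opacity math multiplies an alpha map by the function's output, with parameters for texture depth ratio, explicit depth and softness.
# Color math fakes lighting from a baked lighting map, scaled by the transformed normal map's vertical component and a cubemap reflection.
# A "simple shader" static switch substitutes the lighting map alone.
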Material /Engine/Functions/Engine_MaterialFunctions02/ExampleContent/3dParticleOpacity_Mat
graph {
 properties {
  bUsedWithParticleSprites=true
  blendMode=BLEND_Translucent
  pixelDepthOffsetMode=PDOM_Legacy
  shadingModel=MSM_Unlit
  usedShadingModels=Unlit
 }
 nodes [
  Multiply(
   A=TextureSample() {samplerType=SAMPLERTYPE_LinearColor, texture=/Engine/Functions/Engine_MaterialFunctions02/ExampleContent/Textures/newhplaneAlphaMap}.R
   B=MaterialFunctionCall!(
    Depth Texture
    -------------Depth Texture Options-----------
    "Use dynamic (True) or explicit texture depth settings (False)"
    Dynamic Texture Depth Ratio=ScalarParameter() {defaultValue=0.5, parameterName=Texture Depth to Width Ratio}
    "Explicit Texture Depth (WS)"=ScalarParameter() {defaultValue=80, parameterName=texture depth}
    "Depth Texture Falloff Softness (WS) (1/n)"=ScalarParameter() {defaultValue=0.0625, parameterName=softness}
    -------Camera Falloff Group-------
    Use Near Camera Falloff
    Near Camera Falloff Start Distance
    "Near Camera Fade Distance (1/n)"
    Use Depth In Camera Falloff Calculation
    -------Particle Alpha Group-------
    Use Particle Alpha
    -------WPO Group-------
    Camera Offset
   ) {materialFunction=/Engine/Functions/Engine_MaterialFunctions02/Texturing/3dParticleOpacity}.Opacity
  )
  n1 = TextureSample() {texture=/Engine/Functions/Engine_MaterialFunctions02/ExampleContent/Textures/newhplaneLightingMap}
  n2 = Transform(None=TextureSample() {samplerType=SAMPLERTYPE_Normal, texture=/Engine/Functions/Engine_MaterialFunctions02/ExampleContent/Textures/newhplaneNormalsMap}.RGB)
  n3 = ConstantBiasScale(None=ComponentMask(None=@n2) {b=true}) {scale=0.25}
  n4 = TextureSampleParameterCube!(UVs=ReflectionVectorWS(CustomWorldNormal=@n2), Apply View MipBias) {parameterName=CubeMap, texture=/Engine/MapTemplates/Sky/DaylightAmbientCubemap}
  StaticSwitchParameter(True=@n1.RGB, False=Add(A=Multiply(A=@n3, B=@n1.RGB), B=Multiply(A=@n4.RGB, B=@n1.RGB))) {parameterName=simple shader}
 ]
}
