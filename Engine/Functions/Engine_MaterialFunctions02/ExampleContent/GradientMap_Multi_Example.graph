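# Example Default Lit material that demonstrates the GradientMap_Multi material function from Engine_MaterialFunctions02.
# A single function call takes a scalar parameter named Index as the zero-based gradient selector and a constant for how many gradients the map holds.
# The greyscale value and gradient texture inputs have nothing connected, and no link from the function to a material output appears.
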Material /Engine/Functions/Engine_MaterialFunctions02/ExampleContent/GradientMap_Multi_Example
graph {
 properties {pixelDepthOffsetMode=PDOM_Legacy, usedShadingModels=Default Lit}
 nodes [MaterialFunctionCall!(
   Greyscale Value To Gradient Map
   "Index (0 Based)"=ScalarParameter() {parameterName=Index}
   Gradient Texture
   Number Of Gradients In the Gradient Map=Constant() {r=3}
  ) {materialFunction=/Engine/Functions/Engine_MaterialFunctions02/Gradients/GradientMap_Multi}]
}
