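# HueShift_Mat is an engine example-content material that demonstrates the HueShift material function.
# A scalar parameter named HueShift drives the function's Hue Shift Percentage input, so the shift can be changed per material instance.
# The material uses Default Lit shading, and the graph shows no other function inputs and no connection to a material output.
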Material /Engine/Functions/Engine_MaterialFunctions02/ExampleContent/HueShift_Mat
graph {
 properties {pixelDepthOffsetMode=PDOM_Legacy, usedShadingModels=Default Lit}
 nodes [MaterialFunctionCall(Hue Shift Percentage=ScalarParameter() {parameterName=HueShift}) {materialFunction=&HueShift}]
}
