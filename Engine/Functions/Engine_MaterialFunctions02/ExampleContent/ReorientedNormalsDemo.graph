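# Masked Default Lit example material demonstrating reoriented normal blending of a detail normal map over a base normal map.
# BlendAngleCorrectedNormals combines the base with the detail normal, which an alpha map fades from flat (0,0,1).
# A parallel branch performs a simple additive XY blend using the same mask.
# Constant diffuse and specular values feed ConvertFromDiffSpec, and a second alpha texture is also sampled.
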Material /Engine/Functions/Engine_MaterialFunctions02/ExampleContent/ReorientedNormalsDemo
graph {
 properties {blendMode=BLEND_Masked, pixelDepthOffsetMode=PDOM_Legacy, usedShadingModels=Default Lit}
 nodes [
  n1 = TextureSample() {samplerType=SAMPLERTYPE_Normal, texture=&/Plane002NormalsMap}
  n2 = TextureSample() {samplerType=SAMPLERTYPE_Normal, texture=/Engine/Functions/Engine_MaterialFunctions02/ExampleContent/Textures/newhplaneNormalsMap}
  n3 = Constant3Vector() {constant=<rgba 1 1 0 0>}
  n4 = TextureSample() {samplerType=SAMPLERTYPE_LinearColor, texture=/Engine/Functions/Engine_MaterialFunctions02/ExampleContent/Textures/newhplaneAlphaMap}
  Add(A=@n1.RGB, B=Multiply(A=Multiply(A=@n2.RGB, B=@n3), B=@n4.R))
  Constant() {r=0.1}
  MaterialFunctionCall(BaseNormal=@n1.RGB, AdditionalNormal=LinearInterpolate(A=Constant3Vector() {constant=<rgba 0 0 1 0>}, B=@n2.RGB, Alpha=@n4.R)) {materialFunction=&BlendAngleCorrectedNormals}
  MaterialFunctionCall(DiffuseColor=Constant(), SpecularColor=Constant() {r=0.5}) {materialFunction=&ConvertFromDiffSpec}
  TextureSample() {texture=&/Plane002AlphaMap}
 ]
}
