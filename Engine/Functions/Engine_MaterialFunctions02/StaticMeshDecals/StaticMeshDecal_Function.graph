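# Material function that draws decal-like projections on surfaces using a 1x1x1 corner-pivot box mesh outside the decal domain.
# It projects world position behind translucency onto the component's axes, divided by object scale, for 0-1 UVW.
# Sphere-mask box bounds, a camera clip mask and a scene-normal angle dot product multiply into a combined mask.
# Mobile substitutes a constant normal, and the clamped falloff power node feeds no output.
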MaterialFunction /Engine/Functions/Engine_MaterialFunctions02/StaticMeshDecals/StaticMeshDecal_Function
graph {
 properties {
  bExposeToLibrary=true
  description="Use this function with \"1x1x1_Box_Pivot_-XYZ\" to draw decal like projections on to surfaces. The material should use shader domains other than the decal domain."
  libraryCategoriesText [Decals]
 }
 nodes [
  n1 = MaterialFunctionCall() {materialFunction=/Engine/Functions/Engine_MaterialFunctions02/WorldPositionOffset/ComponentBasisVectors}
  n2 = Subtract(
   A=MaterialFunctionCall() {materialFunction=&WorldPositionBehindTranslucency}.Result
   B=MaterialFunctionCall() {materialFunction=&ObjectPivotPoint}.Object Pivot Location
  )
  n3 = DotProduct(A=Normalize(VectorInput=@n1.X), B=@n2)
  n4 = DotProduct(A=Normalize(VectorInput=@n1.Y), B=@n2)
  n5 = Normalize(VectorInput=@n1.Z)
  n6 = Divide(
   A=AppendVector(A=AppendVector(A=@n3, B=@n4), B=DotProduct(A=@n5, B=@n2))
   B=MaterialFunctionCall() {materialFunction=/Engine/Functions/Engine_MaterialFunctions02/WorldPositionOffset/ObjectScale}.Scale XYZ
  )
  FunctionOutput(None=@n6) {outputName=0-1 RGB UVW}
  n7 = MaterialFunctionCall(Float3=@n6) {materialFunction=&BreakOutFloat3Components}
  n8 = FunctionInput() {bUsePreviewValueAsDefault=true, inputName=XYZ Opacity Falloff Hardness, previewValue=<0.9 0.9 0.9 1>}
  n9 = MaterialFunctionCall(Float3=@n8) {materialFunction=&BreakOutFloat3Components}
  n10 = SphereMask(A=@n7.R, B=Constant() {r=0.5}, Hardness=@n9.R) {attenuationRadius=0.5}
  n11 = SphereMask(A=@n7.G, B=Constant() {r=0.5}, Hardness=@n9.G) {attenuationRadius=0.5}
  n12 = Constant() {r=0.5}
  n13 = FunctionOutput(None=Multiply(A=Multiply(A=@n10, B=@n11), B=SphereMask(A=@n7.B, B=@n12, Hardness=@n9.B) {attenuationRadius=0.5})) {outputName=Extent bounds masks, sortPriority=2}
  n14 = CameraPositionWS()
  n15 = WorldPosition()
  n16 = FunctionInput() {bUsePreviewValueAsDefault=true, inputName=Camera Clip Mask Distance, inputType=FunctionInput_Scalar, sortPriority=1}
  n17 = FunctionOutput(None=OneMinus(None=SphereMask(A=@n14, B=@n15.XYZ, Radius=@n16) {hardnessPercent=0})) {outputName=Camera Clip Mask, sortPriority=3}
  n18 = SceneTexture() {sceneTextureId=PPI_WorldNormal}
  n19 = Constant3Vector() {desc=ES2/3.1 does not allow scene texture sampling, constant=<rgba 1 1 1 0>}
  n20 = FunctionInput() {
   bUsePreviewValueAsDefault=true
   inputName="Local Projection Angle Mask Vector And Falloff Power (V4)"
   inputType=FunctionInput_Vector4
   sortPriority=2
   previewValue=<0 0 1 2>
  }
  n21 = ComponentMask(None=@n20) {b=true, g=true, r=true}
  n22 = DotProduct(
   A=ComponentMask(None=FeatureLevelSwitch(Default=@n18.Color, ES2=@n19, ES3_1=@n19, SM4_REMOVED=@n18.Color)) {b=true, g=true, r=true}
   B=Normalize(VectorInput=Transform(None=@n21) {transformSourceType=TRANSFORMSOURCE_Instance})
  )
  FunctionOutput(None=Multiply(A=Multiply(A=@n13, B=@n17), B=FunctionOutput(None=@n22) {outputName=Projection Angle Mask, sortPriority=4})) {outputName=Combined Mask, sortPriority=1}
  Power(Base=Clamp(None=@n22), Exp=ComponentMask(None=@n20) {a=true})
 ]
}
