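# Example translucent, unlit material for the StaticMeshDecal_Function material function, projecting a decal texture through a static mesh.
# The function is given a local Z projection vector with an angle falloff power, and the R and G channels of its UVW output serve as UVs.
# Those UVs sample a plane lighting map and an alpha map, with the alpha multiplied by the function's combined mask to confine the decal.
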Material /Engine/Functions/Engine_MaterialFunctions02/ExampleContent/StaticMeshDecal
graph {
 properties {blendMode=BLEND_Translucent, pixelDepthOffsetMode=PDOM_Legacy, shadingModel=MSM_Unlit, usedShadingModels=Unlit}
 nodes [
  n1 = MaterialFunctionCall!(
   XYZ Opacity Falloff Hardness
   Camera Clip Mask Distance
   "Local Projection Angle Mask Vector And Falloff Power (V4)"=Constant4Vector() {constant=<rgba 0 0 1 8>}.RGBA
  ) {materialFunction=/Engine/Functions/Engine_MaterialFunctions02/StaticMeshDecals/StaticMeshDecal_Function}
  n2 = ComponentMask(None=@n1.0-1 RGB UVW) {g=true, r=true}
  Multiply(
   A=TextureSample(UVs=@n2) {samplerType=SAMPLERTYPE_LinearColor, texture=/Engine/Functions/Engine_MaterialFunctions02/ExampleContent/Textures/newhplaneAlphaMap}.RGB
   B=@n1.Combined Mask
  )
  TextureSample(UVs=@n2) {texture=/Engine/Functions/Engine_MaterialFunctions02/ExampleContent/Textures/newhplaneLightingMap}
 ]
}
