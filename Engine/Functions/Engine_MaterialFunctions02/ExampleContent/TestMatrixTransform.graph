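# Example material that tests the InverseTransformMatrix material function by animating texture UVs.
# TimeWithSpeedVariable drives normalized lerps of the X and Y basis vectors through a repeating quarter turn, with a constant W offset added to Y.
# UVs padded to a homogeneous four-component vector pass through the inverse transform and are masked to RG.
# The resulting coordinates sample lighting and alpha maps, whose RGB values are multiplied together.
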Material /Engine/Functions/Engine_MaterialFunctions02/ExampleContent/TestMatrixTransform
graph {
 properties {pixelDepthOffsetMode=PDOM_Legacy, usedShadingModels=Default Lit}
 nodes [
  n1 = TextureCoordinate()
  n2 = Constant2Vector() {g=1}
  n3 = Constant4Vector()
  n4 = MaterialFunctionCall(Speed=Constant() {r=0.25}) {materialFunction=&TimeWithSpeedVariable}
  n5 = Normalize(VectorInput=LinearInterpolate(A=Constant4Vector() {constant=<rgba 0 1 0 0>}.RGBA, B=Constant4Vector() {constant=<rgba 1 0 0 0>}.RGBA, Alpha=@n4.Frac Time))
  n6 = Constant4Vector() {constant=<rgba 0 0 0 0.5>}
  n7 = Normalize(VectorInput=LinearInterpolate(A=Constant4Vector() {constant=<rgba -1 0 0 0>}.RGBA, B=Constant4Vector() {constant=<rgba 0 1 0 0>}.RGBA, Alpha=@n4.Frac Time))
  n8 = ComponentMask(None=MaterialFunctionCall!(VectorToTransform=AppendVector(A=@n1, B=@n2), BasisX=Add(A=@n3.RGBA, B=@n5), BasisY=Add(A=@n6.RGBA, B=@n7), BasisZ) {materialFunction=/Engine/Functions/Engine_MaterialFunctions02/Math/InverseTransformMatrix}.Transformed Vector 3) {g=true, r=true}
  Multiply(
   A=TextureSample(UVs=@n8) {texture=/Engine/Functions/Engine_MaterialFunctions02/ExampleContent/Textures/newhplaneLightingMap}.RGB
   B=TextureSample(UVs=@n8) {samplerType=SAMPLERTYPE_LinearColor, texture=/Engine/Functions/Engine_MaterialFunctions02/ExampleContent/Textures/newhplaneAlphaMap}.RGB
  )
 ]
}
