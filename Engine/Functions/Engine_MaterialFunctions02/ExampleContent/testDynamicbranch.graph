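# Example Default Lit material that exercises the engine's DynamicBranch material function.
# A scalar parameter drives the branch Alpha, with Color Input 0 left unconnected and Color Input 1 fed by a feature level switch.
# The switch supplies Value noise by default and substitutes a cheap constant on ES2 and ES3_1.
# The graph shows no link from the branch result to a material output.
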Material /Engine/Functions/Engine_MaterialFunctions02/ExampleContent/testDynamicbranch
graph {
 properties {pixelDepthOffsetMode=PDOM_Legacy, usedShadingModels=Default Lit}
 nodes [
  n1 = Noise!(World Position, FilterWidth) {noiseFunction=NOISEFUNCTION_ValueALU}
  n2 = Constant()
  MaterialFunctionCall!(Alpha=ScalarParameter(), Color Input 0, Color Input 1=FeatureLevelSwitch(Default=@n1, ES2=@n2, ES3_1=@n2, SM4_REMOVED=@n1)) {materialFunction=/Engine/Functions/Engine_MaterialFunctions02/Utility/DynamicBranch}
 ]
}
